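# Sound cue for collecting a shard in the Season 21 Week 5 narrative quest.
# A Random node picks one of three CollectShard wave variants with equal weight.
# The cue uses the week's shared attenuation asset and plays through the Fort_Quests sound class.
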SoundCue /BattlePassS21/QuestActors/Narrative/Week_05/SFX/S21_W05_CollectShard_Cue
root=Random_0
nodes [
 {id=Random_0, children [WavePlayer_3,WavePlayer_4,WavePlayer_5], properties {weights [1,1,1]}}
 {id=WavePlayer_3, properties {soundWaveAssetPtr=&/S21_W05_CollectShard_01}}
 {id=WavePlayer_4, properties {soundWaveAssetPtr=&/S21_W05_CollectShard_02}}
 {id=WavePlayer_5, properties {soundWaveAssetPtr=&/S21_W05_CollectShard_03}}
]
properties {
 attenuationSettings=&/S21_W05_Atten
 maxDistance=1300
 soundClassObject=&Fort_Quests
}
