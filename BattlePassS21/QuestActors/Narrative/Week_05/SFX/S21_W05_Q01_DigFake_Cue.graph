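# S21_W05_Q01_DigFake_Cue is a sound cue for fake-dig sound effects in the Season 21 Week 5 narrative quest.
# A Random node picks among five equally weighted DigFake wave variations each time it plays.
# Playback routes through the quests sound class and uses the shared week 5 attenuation asset for distance falloff.
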SoundCue /BattlePassS21/QuestActors/Narrative/Week_05/SFX/S21_W05_Q01_DigFake_Cue
root=Random_0
nodes [
 {id=Random_0, children [WavePlayer_0,WavePlayer_1,WavePlayer_2,WavePlayer_3,WavePlayer_4], properties {weights [1,1,1,1,1]}}
 {id=WavePlayer_0, properties {soundWaveAssetPtr=&/S21_W05_Q01_DigFake_01}}
 {id=WavePlayer_1, properties {soundWaveAssetPtr=&/S21_W05_Q01_DigFake_02}}
 {id=WavePlayer_2, properties {soundWaveAssetPtr=&/S21_W05_Q01_DigFake_03}}
 {id=WavePlayer_3, properties {soundWaveAssetPtr=&/S21_W05_Q01_DigFake_04}}
 {id=WavePlayer_4, properties {soundWaveAssetPtr=&/S21_W05_Q01_DigFake_05}}
]
properties {
 attenuationSettings=&/S21_W05_Atten
 maxDistance=1300
 soundClassObject=&Fort_Quests
 volumeMultiplier=1
}
