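# Sound cue for the shard-digging sound in a Season 21 Battle Pass Week 5 narrative quest.
# A Random node picks one of three equally weighted DigShard wave variations each time it plays.
# It routes through the Fort_Quests sound class and uses the shared S21_W05 attenuation asset for distance falloff.
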SoundCue /BattlePassS21/QuestActors/Narrative/Week_05/SFX/S21_W05_Q01_DigShard_Cue
root=Random_0
nodes [
 {id=Random_0, children [WavePlayer_0,WavePlayer_1,WavePlayer_2], properties {weights [1,1,1]}}
 {id=WavePlayer_0, properties {soundWaveAssetPtr=&/S21_W05_Q01_DigShard_01}}
 {id=WavePlayer_1, properties {soundWaveAssetPtr=&/S21_W05_Q01_DigShard_02}}
 {id=WavePlayer_2, properties {soundWaveAssetPtr=&/S21_W05_Q01_DigShard_03}}
]
properties {
 attenuationSettings=&/S21_W05_Atten
 maxDistance=1300
 soundClassObject=&Fort_Quests
 volumeMultiplier=1
}
