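# Looping sound cue for pipe water in a Season 21 Week 05 Battle Pass quest (Q02).
# A Random node picks one of four looping pipe-water waves with equal weight, feeding a Modulator that adds slight volume variation and leaves pitch fixed.
# Playback uses the shared Week 05 attenuation asset and the Fort_Quests sound class.
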SoundCue /BattlePassS21/QuestActors/Narrative/Week_05/SFX/S21_W05_Q02_PipeWater_LP_Cue
root=Modulator_0
nodes [
 {id=Modulator_0, children [Random_0], properties {pitchMax=1, pitchMin=1, volumeMax=1.02, volumeMin=0.98}}
 {id=Random_0, children [WavePlayer_0,WavePlayer_1,WavePlayer_2,WavePlayer_3], properties {weights [1,1,1,1]}}
 {id=WavePlayer_0, properties {bLooping=true, soundWaveAssetPtr=&/S21_W05_Q02_PipeWater_LP_01}}
 {id=WavePlayer_1, properties {bLooping=true, soundWaveAssetPtr=&/S21_W05_Q02_PipeWater_LP_02}}
 {id=WavePlayer_2, properties {bLooping=true, soundWaveAssetPtr=&/S21_W05_Q02_PipeWater_LP_03}}
 {id=WavePlayer_3, properties {bLooping=true, soundWaveAssetPtr=&/S21_W05_Q02_PipeWater_LP_04}}
]
properties {
 attenuationSettings=&/S21_W05_Atten
 maxDistance=1300
 soundClassObject=&Fort_Quests
 volumeMultiplier=1
}
