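# Sound cue for the pressure-release effect in the Battle Pass Season 21, Week 5, Quest 2 narrative quest actors.
# A Random node picks one of three equally weighted ReleasePressure waves and feeds a Mixer root, whose second input is left empty.
# Playback uses the shared Week 5 attenuation settings and the Fort_Quests sound class.
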SoundCue /BattlePassS21/QuestActors/Narrative/Week_05/SFX/S21_W05_Q02_ReleasePressure_Cue
root=Mixer_0
nodes [
 {id=Mixer_0, children [Random_0,null], properties {inputVolume [1,0.6]}}
 {id=Random_0, children [WavePlayer_0,WavePlayer_1,WavePlayer_2], properties {weights [1,1,1]}}
 {id=WavePlayer_0, properties {soundWaveAssetPtr=&/S21_W05_Q02_ReleasePressure_01}}
 {id=WavePlayer_1, properties {soundWaveAssetPtr=&/S21_W05_Q02_ReleasePressure_02}}
 {id=WavePlayer_2, properties {soundWaveAssetPtr=&/S21_W05_Q02_ReleasePressure_03}}
]
properties {
 attenuationSettings=&/S21_W05_Atten
 maxDistance=1300
 soundClassObject=&Fort_Quests
 volumeMultiplier=1
}
