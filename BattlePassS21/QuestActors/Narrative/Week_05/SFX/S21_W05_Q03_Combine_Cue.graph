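# Sound cue for the combine sound in a Battle Pass Season 21, Week 5, Quest 3 narrative quest actor.
# A single wave player feeds the S21_W05_Q03_Combine wave straight to the output, with no randomization or modulation.
# The cue plays in the Fort_Quests sound class and uses the shared Week 5 attenuation asset for distance falloff.
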SoundCue /BattlePassS21/QuestActors/Narrative/Week_05/SFX/S21_W05_Q03_Combine_Cue
root=WavePlayer_0
nodes [{id=WavePlayer_0, properties {soundWaveAssetPtr=&/S21_W05_Q03_Combine}}]
properties {
 attenuationSettings=&/S21_W05_Atten
 maxDistance=1300
 soundClassObject=&Fort_Quests
 volumeMultiplier=1
}
