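# Looping light ray sound cue for the Season 21 Week 5 Q03 narrative quest actor, played under the Fort_Quests sound class.
# A random node picks one of three equally weighted looping waves, then a modulator applies slight pitch variation.
# Overridden attenuation gives a short-range falloff with occlusion filtering and listener focus, so the ray grows quieter when the listener is not facing it.
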SoundCue /BattlePassS21/QuestActors/Narrative/Week_05/SFX/S21_W05_Q03_LightRay_Cue
root=Modulator_0
nodes [
 {id=Modulator_0, children [Random_0], properties {pitchMax=1.01, pitchMin=0.99}}
 {id=Random_0, children [WavePlayer_0,WavePlayer_1,WavePlayer_2], properties {weights [1,1,1]}}
 {id=WavePlayer_0, properties {bLooping=true, soundWaveAssetPtr=&/S21_W05_Q03_LightRay_LP_01}}
 {id=WavePlayer_1, properties {bLooping=true, soundWaveAssetPtr=&/S21_W05_Q03_LightRay_LP_02}}
 {id=WavePlayer_2, properties {bLooping=true, soundWaveAssetPtr=&/S21_W05_Q03_LightRay_LP_03}}
]
properties {
 attenuationSettings=&/S21_W05_Atten
 bOverrideAttenuation=true
 maxDistance=900
 soundClassObject=&Fort_Quests
 volumeMultiplier=0.9
 attenuationOverrides {
  bEnableFocusInterpolation=true
  bEnableListenerFocus=true
  falloffDistance=800
  focusAttackInterpSpeed=0.2
  focusAzimuth=15
  focusReleaseInterpSpeed=0.2
  nonFocusAzimuth=30
  nonFocusDistanceScale=0.7
  nonFocusVolumeAttenuation=0.4
  occlusionLowPassFilterFrequency=500
  occlusionVolumeAttenuation=0.6
  attenuationShapeExtents {x=100}
 }
}
