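# Sound cue for the looping shard sound in the Battle Pass S21 Week 5 narrative quest actors.
# A single looping wave player plays the S21_W05_Shard_LP sound wave directly as the cue's root, with no mixing or randomization.
# It routes to the Fort_Quests sound class at reduced volume and uses the S21_W05_Atten attenuation for spatial falloff.
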SoundCue /BattlePassS21/QuestActors/Narrative/Week_05/SFX/S21_W05_Shard_LP_Cue
root=WavePlayer_0
nodes [{id=WavePlayer_0, properties {bLooping=true, soundWaveAssetPtr=&/S21_W05_Shard_LP}}]
properties {
 attenuationSettings=&/S21_W05_Atten
 maxDistance=1300
 soundClassObject=&Fort_Quests
 volumeMultiplier=0.6
}
